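; Parent gameplay ability that applies effective healing and shielding to its avatar pawn, once or on a repeating tick with an optional duration.
; Each heal clamps amounts to missing health/shield and optional caps, can overflow excess heal into shields, then applies GE_Heal via SetByCaller magnitudes.
; It also fires gameplay cues, quest stats and optional healing-magnitude events.
; Children can override the CheckCanHeal gate and cue functions.
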(event Ability|EventActivateAbility
  ;[1 1 1] Activate Ability
  (|CallParentFunction)
  (Utilities|FlowControl|Sequence
    (:then_0
      (bind _asfort_pawn (Utilities|Casting|CastToFortPawn (Ability|GetAvatarActorfromActorInfo))
        (:then
          (Variables|Default|SetPawn _asfort_pawn))
        (:CastFailed
          (Ability|EndAbility))))
    (:then_1
      (Variables|Default|SetEndAbilityAfterHeal false)
      (if (> (ScalableFloat|GetValueatLevel (Variables|Balance|GetDuration)) 0.0)
        (Utilities|Time|SetTimerbyEvent (delegate Custom|DurationExpired) (ScalableFloat|GetValueatLevel (Variables|Balance|GetDuration)))))
    (:then_2
      (if (> (ScalableFloat|GetValueatLevel (Variables|Balance|GetTickTime)) 0.0)
        (Utilities|Time|SetTimerbyEvent (EventDispatchers|CreateEvent :!event Custom|Heal) (ScalableFloat|GetValueatLevel (Variables|Balance|GetTickTime)) true)
        (if (ScalableFloat|GetValueasBool (Variables|Balance|GetHealTickOnApplication))
          (CallFunction|Heal))
        (else
          (Variables|Default|SetEndAbilityAfterHeal true)
          (CallFunction|Heal))))))

(event Custom|DurationExpired
  ;[1 1 1] Activate Ability
  (Ability|EndAbility))

(event Custom|Heal
  (bind _pawn (Variables|Default|GetPawn))
  (bind _returnvalue (Fortnite|GetFortPlayerControllerfromActor _pawn))
  (bind _returnvalue_2 (Ability|GetGrantedbyEffectContext))
  (bind _returnvalue_3 (Ability|EffectContext|GetEffectCauser _returnvalue_2))
  (bind _returnvalue_4 (Ability|EffectContext|GetSourceObject _returnvalue_2))
  (bind _returnvalue_5 (Ability|EffectContext|GetOriginalInstigatorActor _returnvalue_2))
  (bind _pawn_2 (Variables|Default|GetPawn))
  (bind _pawn_3 (Variables|Default|GetPawn))
  (bind _event_payloadinstigatortags (Variables|Default|GetEvent_PayloadInstigatorTags))
  ;[0.191 1 0.5347] Heal
  (Utilities|IsValid (Variables|Default|GetPawn)
    (:Is%20Valid
      (if (CallFunction|CheckCanHeal)
        (bind (_healmagnitude _shieldmagnitude _applyeffect _restoredhealth _restoredshields) (CallFunction|GetMagnitudes))
        (Variables|Default|SetHealMagnitude _healmagnitude)
        (Variables|Default|SetShieldMagnitude _shieldmagnitude)
        (bind _returnvalue_6 (Ability|GameplayEffect|AssignTagSetbyCallerMagnitude (GameplayEffects|MakeOutgoingSpec (Ability|GetAbilitySystemComponent (Variables|Default|GetPawn)) (Variables|Default|GetGE_Heal) 1.0 (CallFunction|MakeHealEffectContext)) "(TagName=\"SetByCaller.Healing\")" (Variables|Default|GetHealMagnitude)))
        (bind _returnvalue_7 (Ability|GameplayEffect|AssignTagSetbyCallerMagnitude _returnvalue_6 "(TagName=\"SetByCaller.Shield\")" (Variables|Default|GetShieldMagnitude)))
        (if _applyeffect
          (Ability|ApplyGameplayEffectSpecToOwner _returnvalue_7)
          (else
            _))
        (if (and (GameplayTags|NotEqual(GameplayTag) :A (Variables|GameplayCues|GetTag_GC_Heal)) _restoredhealth)
          (CallFunction|ExecuteGameplayCueHealHealth)
          (else
            _))
        (if (and (GameplayTags|NotEqual(GameplayTag) :A (Variables|GameplayCues|GetTag_GC_Shield)) _restoredshields)
          (CallFunction|ExecuteGameplayCueHealShields)
          (else
            _))
        (Utilities|IsValid _returnvalue
          (:Is%20Valid
            (Quest|SendComplexCustomStatEvent (Quests|GetQuestManager _returnvalue "Invalid") 0 (Variables|Quest|GetQuestSourceTags) (Variables|Quest|GetQuestTargetTags) 1))
          (:Is%20Not%20Valid
            _))
        (if (Variables|Default|GetSendHealingMagnitudeQuestEvents)
          (Ability|SendGameplayEventtoActor _pawn_2 "(TagName=\"Event.Ability.EffectiveHeal.Health\")" (Utilities|Struct|MakeGameplayEventData "(TagName=\"\")" _returnvalue_5 _pawn_3 _returnvalue_3 _returnvalue_4 "()" _event_payloadinstigatortags "(GameplayTags=)" (Variables|Default|GetHealMagnitude)))
          (Ability|SendGameplayEventtoActor _pawn_2 "(TagName=\"Event.Ability.EffectiveHeal.Shields\")" (Utilities|Struct|MakeGameplayEventData "(TagName=\"\")" _returnvalue_5 _pawn_3 _returnvalue_3 _returnvalue_4 "()" _event_payloadinstigatortags "(GameplayTags=)" (Variables|Default|GetShieldMagnitude)))
          (else
            _))
        (if (Variables|Default|GetEndAbilityAfterHeal)
          (Utilities|FlowControl|DelayUntilNextTick)
          (Ability|EndAbility))))
    (:Is%20Not%20Valid
      (Ability|EndAbility))))

(fn CheckCanHeal
  ;[1 1 1] Can be overridden by children
  (return true))

(fn ExecuteGameplayCue_HealHealth
  ;[1 1 1] May be overridden by children
  (Ability|ExecuteGameplayCueOnOwner (Variables|GameplayCues|GetTag_GC_Heal)))

(fn ExecuteGameplayCue_HealShields
  ;[1 1 1] May be overridden by children
  (Ability|ExecuteGameplayCueOnOwner (Variables|GameplayCues|GetTag_GC_Shield)))

(fn GetMagnitudes
  (bind _pawn (Variables|Default|GetPawn))
  (LocalVariables|SetCappedHealAmount (Math|Float|Max(Float) (ScalableFloat|GetValueatLevel (Variables|Balance|GetHeal_Amount))))
  (LocalVariables|SetCappedShieldAmount (Math|Float|Max(Float) (ScalableFloat|GetValueatLevel (Variables|Balance|GetShield_Amount))))
  (LocalVariables|SetMissingHealth (* (Math|Conversions|ToFloat(Integer) (Math|Float|Round (* (- (FortPawn|GetMaxHealth _pawn) (FortPawn|GetHealth _pawn)) 1e+02))) 0.01))
  (LocalVariables|SetMissingShield (* (Math|Conversions|ToFloat(Integer) (Math|Float|Round (* (- (FortPawn|GetMaxShield _pawn) (FortPawn|GetShield _pawn)) 1e+02))) 0.01))
  (if (<= (ScalableFloat|GetValueatLevel (Variables|Balance|GetHeal_Cap)) 0.0)
    (LocalVariables|SetCappedHealAmount (Math|Float|Min(Float) (LocalVariables|GetCappedHealAmount) (LocalVariables|GetMissingHealth)))
    (else
      (LocalVariables|SetCappedHealAmount (Math|Float|Min(Float) (LocalVariables|GetCappedHealAmount) (Math|Float|Max(Float) (- (ScalableFloat|GetValueatLevel (Variables|Balance|GetHeal_Cap)) (FortPawn|GetHealth (Variables|Default|GetPawn)))) (LocalVariables|GetMissingHealth)))))
  (if (ScalableFloat|GetValueasBool (Variables|Balance|GetbHealOverflowsIntoShields))
    (LocalVariables|SetOverHealAmount (- (Math|Float|Max(Float) (ScalableFloat|GetValueatLevel (Variables|Balance|GetHeal_Amount))) (LocalVariables|GetCappedHealAmount)))
    (else
      (LocalVariables|SetOverHealAmount 0.0)))
  (LocalVariables|SetCappedShieldAmount (+ (LocalVariables|GetCappedShieldAmount) (LocalVariables|GetOverHealAmount)))
  (if (<= (ScalableFloat|GetValueatLevel (Variables|Balance|GetShield_Cap)) 0.0)
    (LocalVariables|SetCappedShieldAmount (Math|Float|Min(Float) (LocalVariables|GetCappedShieldAmount) (LocalVariables|GetMissingShield)))
    (else
      (LocalVariables|SetCappedShieldAmount (Math|Float|Min(Float) (LocalVariables|GetCappedShieldAmount) (Math|Float|Max(Float) (- (ScalableFloat|GetValueatLevel (Variables|Balance|GetShield_Cap)) (FortPawn|GetShield (Variables|Default|GetPawn)))) (LocalVariables|GetMissingShield)))))
  (return (LocalVariables|GetCappedHealAmount) (LocalVariables|GetCappedShieldAmount) (or (> (LocalVariables|GetCappedHealAmount) 0.0) (> (LocalVariables|GetCappedShieldAmount) 0.0)) (> (LocalVariables|GetCappedHealAmount) 0.0) (> (LocalVariables|GetCappedShieldAmount) 0.0)))

(fn MakeHealEffectContext
  (return (Class|AthenaFunctionLibrary|MakeSimpleFortEffectContext (Ability|EffectContext|GetInstigatorActor (Ability|GetGrantedbyEffectContext)))))
